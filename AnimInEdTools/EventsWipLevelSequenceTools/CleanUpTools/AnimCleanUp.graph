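# Editor utility widget whose single button cleans up animation sequences used by the focused Level Sequence and its subsequences.
# FindSubSequence recursively walks sub-tracks, while GetSequenceAnims passes each skeletal animation section's asset to FixAnimData.
# FixAnimData enforces Fortnite bone compression settings and follows the animation's link to its source Level Sequence.
# A source not at 30 fps is reset to 30, then saved along with its linked animations.
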EditorUtilityWidgetBlueprint /AnimInEdTools/EventsWipLevelSequenceTools/CleanUpTools/AnimCleanUp
graphs [
 EventGraph {nodes [
   "On Clicked (EditorUtilityButton_28)"() {is=title K2Node_ComponentBoundEvent}
   > n1 = GetFocusedLevelSequence!(execute) -> (then, ReturnValue:Level Sequence Object Reference)
   > FindSubSequence!(execute, self, Sequence:Movie Scene Sequence Object Reference=@n1) -> (then)
   Event PreConstruct()
   Event Construct()
   Event Tick()
  ]}
 FindSubSequence {nodes [
   n1 = FindSubSequence!() -> (then, Sequence:Movie Scene Sequence Object Reference) {is=title K2Node_FunctionEntry}
   > n2 = GetTracks!(execute, Sequence:Movie Scene Sequence Object Reference=@n1) -> (then, ReturnValue:Array of Movie Scene Track Object References)
   > n3 = For Each Loop(Array:Array of Movie Scene Track Object References=@n2) -> (Array Element:Movie Scene Track Object Reference)
   GetSequenceAnims!(execute=@n3.Completed, self, Sequence:Movie Scene Sequence Object Reference=@n1) -> (then)
   > n4 = For Each Loop(Array:Array of Movie Scene Sequence Object References=Get SubSequences() -> (SubSequences:Array of Movie Scene Sequence Object References)) -> (Array Element:Movie Scene Sequence Object Reference)
   > FindSubSequence(Sequence=@n4) {is=title K2Node_CallFunction}
   n5 = Cast To MovieSceneSubTrack!(execute=@n3.LoopBody, Object=@n3) -> (then, CastFailed, AsMovie Scene Sub Track:Movie Scene Sub Track Object Reference) {is=title K2Node_DynamicCast}
   > n6 = GetSections!(execute, Track:Movie Scene Track Object Reference=@n5) -> (then, ReturnValue:Array of Movie Scene Section Object References)
   > n7 = For Each Loop(Array:Array of Movie Scene Section Object References=@n6) -> (Array Element:Movie Scene Section Object Reference)
   > n8 = Cast To MovieSceneSubSection!(execute, Object=@n7) -> (then, CastFailed, AsMovie Scene Sub Section:Movie Scene Sub Section Object Reference) {is=title K2Node_DynamicCast}
   > Add Unique(
    TargetArray:Array of Movie Scene Sequence Object References=Get SubSequences()
    NewItem:"Movie Scene Sequence Object Reference (by ref)"=GetSequence!(self:Movie Scene Sub Section Object Reference=@n8) -> (ReturnValue:Movie Scene Sequence Object Reference)
   )
  ]}
 FixAnimData {nodes [
   n1 = FixAnimData!() -> (then, AnimSeq:Anim Sequence Base Object Reference) {is=title K2Node_FunctionEntry}
   > n2 = Cast To AnimSequence!(execute, Object=@n1) -> (then, CastFailed, AsAnim Sequence:Anim Sequence Object Reference)
   > Set AnimSeqRef(AnimSeqRef:Anim Sequence Object Reference=@n2) -> (Output_Get:Anim Sequence Object Reference)
   > n3 = Sequence()
   n4 = GetLevelSequenceLinkFromAnimSequence!(execute=@n3.then_1, InAnimSequence:Anim Sequence Object Reference=Get AnimSeqRef() -> (AnimSeqRef:Anim Sequence Object Reference)) -> (then, ReturnValue:Anim Sequence Level Sequence Link Object Reference)
   > n5 = Get Asset By Object Path!(
    execute
    self:Asset Registry Interface=GetAssetRegistry()
    ObjectPath_PathString:String=Get!(self:Anim Sequence Level Sequence Link Object Reference=@n4) -> (PathToLevelSequence_PathString:String)
    bIncludeOnlyOnDiskAssets:bool=false
    bSkipARFilteredAssets:bool=true
   ) -> (then, ReturnValue:Asset Data Structure) {is=title K2Node_CallFunction}
   > n6 = Cast To LevelSequence!(execute, Object=GetAsset(InAssetData=@n5)) -> (then, CastFailed, AsLevel Sequence:Level Sequence Object Reference) {is=title K2Node_DynamicCast}
   > n7 = Set LinkedLevelSeq(LinkedLevelSeq:Level Sequence Object Reference=@n6) -> (Output_Get:Level Sequence Object Reference)
   > n8 = GetDisplayRate!(execute, Sequence:Movie Scene Sequence Object Reference=@n7) -> (then, ReturnValue_Numerator:int=0, ReturnValue_Denominator:int=0)
   > n9 = Branch(Condition="Equal (Integer)"(A=@n8, B=30))
   SetDisplayRate!(
    execute=@n9.else
    Sequence:Movie Scene Sequence Object Reference=Get LinkedLevelSeq() -> (LinkedLevelSeq:Level Sequence Object Reference)
    DisplayRate_Numerator:int=30
    DisplayRate_Denominator:int=1
   ) -> (then)
   > SaveLoadedAsset(AssetToSave=Get LinkedLevelSeq(), bOnlyIfIsDirty=true)
   > n10 = GetAnimSequenceLinkFromLevelSequence!(execute, InLevelSequence:Level Sequence Object Reference=Get LinkedLevelSeq()) -> (then, ReturnValue:Level Sequence Anim Sequence Link Object Reference)
   > n11 = For Each Loop!(
    Exec
    Array:Array of Level Sequence Anim Sequence Link Item Structures=Get AnimSequenceLinks!(self:Level Sequence Anim Sequence Link Object Reference=@n10) -> (AnimSequenceLinks:Array of Level Sequence Anim Sequence Link Item Structures)
   ) -> (
    LoopBody
    Array Element_SkelTrackGuid:Guid Structure
    Array Element_PathToAnimSequence:Soft Object Path Structure
    Array Element_bAutoBake:bool=~
    Array Element_bExportTransforms:bool=~
    Array Element_bExportMorphTargets:bool=~
    Array Element_bExportAttributeCurves:bool=~
    Array Element_bExportMaterialCurves:bool=~
    Array Element_bSkipCurvesWithZeroValue:bool=~
    Array Element_bSetRetargetSourceAsset:bool=~
    Array Element_Interpolation:EAnimInterpolationType Enum
    Array Element_CurveInterpolation:ERichCurveInterpMode Enum
    Array Element_bRecordInWorldSpace:bool=~
    Array Element_bEvaluateAllSkeletalMeshComponents:bool=~
    Array Element_IncludeAnimationNames:Array of Strings
    Array Element_ExcludeAnimationNames:Array of Strings
    Array Element_bRemoveExcludedCurves:bool=~
    Array Element_WarmUpFrames:Frame Number Structure
    Array Element_DelayBeforeStart:Frame Number Structure
    Array Element_bUseCustomTimeRange:bool=~
    Array Element_CustomStartFrame:Frame Number Structure
    Array Element_CustomEndFrame:Frame Number Structure
    Array Element_CustomDisplayRate:Frame Rate Structure
    Array Element_bUseCustomFrameRate:bool=~
    Array Element_CustomFrameRate:Frame Rate Structure
    Array Index
    Completed
   )
   > n12 = LoadAsset(AssetPath=BreakSoftObjectPath(InSoftObjectPath=@n11.Array Element_PathToAnimSequence))
   > n13 = Cast To AnimSequence(Object=@n12)
   > SaveLoadedAsset(AssetToSave=@n13, bOnlyIfIsDirty=true)
   n14 = Get AnimSeqRef()
   n15 = Branch(
    execute=@n3.then_0
    Condition="Equal (Object)"(
     A=GetBoneCompressionSettings!(AnimationSequence:Anim Sequence Object Reference=@n14) -> (CompressionSettings:Anim Bone Compression Settings Object Reference)
     B=/Game/Animation/Settings/FortniteAnimBoneCompressionSettings
    )
   )
   SetBoneCompressionSettings!(
    execute=@n15.else
    AnimationSequence:Anim Sequence Object Reference=Get AnimSeqRef()
    CompressionSettings:Anim Bone Compression Settings Object Reference=&FortniteAnimBoneCompressionSettings
   ) -> (then)
   > SaveLoadedAsset(AssetToSave=Get AnimSeqRef(), bOnlyIfIsDirty=true)
  ]}
 GetSequenceAnims {nodes [
   n1 = GetSequenceAnims!() -> (then, Sequence:Movie Scene Sequence Object Reference) {is=title K2Node_FunctionEntry}
   > n2 = GetBindings!(execute, Sequence:Movie Scene Sequence Object Reference=@n1) -> (then, ReturnValue:Array of Movie Scene Binding Proxy Structures)
   > n3 = For Each Loop!(Exec, Array:Array of Movie Scene Binding Proxy Structures=@n2) -> (LoopBody, Array Element:Movie Scene Binding Proxy Structure, Array Index, Completed)
   > n4 = GetTracks!(execute, InBinding:"Movie Scene Binding Proxy Structure (by ref)"=@n3) -> (then, ReturnValue)
   > n5 = For Each Loop(Array:Array of Movie Scene Track Object References=@n4) -> (Array Element:Movie Scene Track Object Reference)
   > n6 = Cast To MovieSceneSkeletalAnimationTrack!(execute, Object=@n5) -> (then, CastFailed, AsMovie Scene Skeletal Animation Track:Movie Scene Skeletal Animation Track Object Reference) {is=title K2Node_DynamicCast}
   > n7 = GetSections(Track=@n6)
   > n8 = For Each Loop(Array:Array of Movie Scene Section Object References=@n7) -> (Array Element:Movie Scene Section Object Reference)
   > n9 = Cast To MovieSceneSkeletalAnimationSection!(execute, Object=@n8) -> (then, CastFailed, AsMovie Scene Skeletal Animation Section:Movie Scene Skeletal Animation Section Object Reference) {is=title K2Node_DynamicCast}
   > FixAnimData!(
    execute
    self
    AnimSeq:Anim Sequence Base Object Reference=Get!(self:Movie Scene Skeletal Animation Section Object Reference=@n9) -> (
     Params_Animation:Anim Sequence Base Object Reference
     Params_FirstLoopStartFrameOffset:Frame Number Structure
     Params_StartFrameOffset:Frame Number Structure
     Params_EndFrameOffset:Frame Number Structure
     Params_PlayRate:Movie Scene Time Warp Variant Structure
     Params_bReverse:bool=~
     Params_SlotName:Name
     Params_MirrorDataTable:Mirror Data Table Object Reference
     Params_bSkipAnimNotifiers:bool=~
     Params_bForceCustomMode:bool=~
     Params_SwapRootBone:ESwapRootBone Enum
     Params_bLinearPlaybackWhenScaled:bool=~
    )
   ) -> (then) {is=title K2Node_CallFunction}
  ]}
]
variables [anim {type=AnimSequence, default=null}]
widgets {
 info {rootWidgetClass=&EditorUtilityButton, widgetCount=2}
 widgets [
  EditorUtilityButton_28 {
   widgetClassPath=&EditorUtilityButton
   widget=EditorUtilityButton_28
   properties {
    toolTipText="This will cycle through the currently focused level sequence, all sub sequences, and finding any anim sequence. It will then check and set the compression setting if incorrect, and will check to see if the anim sequence has a linked level sequence. If that linked level sequence is not 30 FPS it will be set to 30 FPS. Any file modified will be checked out and saved."
    backgroundColor=<rgba 1 0 0 1>
    widgetStyle {
     clickedSlateSound {resourceObject=null}
     disabled {
      bIsDynamicallyLoaded=false
      drawAs=RoundedBox
      imageType=NoImage
      mirroring=NoMirror
      resourceName=None
      resourceObject=null
      tiling=NoTile
      imageSize=<0 0>
      margin=<ltrb 0 0 0 0>
      outlineSettings {
       bUseBrushTransparency=false
       roundingType=FixedRadius
       width=1
       color {colorUseRule=UseColor_ColorTable, specifiedColor=<rgba 0 0 0 0>}
       cornerRadii=<4 4 4 4>
      }
      tintColor {colorUseRule=UseColor_Specified, specifiedColor=<rgba 0.0395 0.0395 0.0395 1>}
      uVRegion {bIsValid=false, max=<0 0>, min=<0 0>}
     }
     disabledForeground {colorUseRule=UseColor_ColorTable, specifiedColor=<rgba 0 0 0 0>}
     hovered {
      bIsDynamicallyLoaded=false
      drawAs=RoundedBox
      imageType=NoImage
      mirroring=NoMirror
      resourceName=None
      resourceObject=null
      tiling=NoTile
      imageSize=<32 32>
      margin=<ltrb 0 0 0 0>
      outlineSettings {
       bUseBrushTransparency=false
       roundingType=FixedRadius
       width=1
       color {colorUseRule=UseColor_ColorTable, specifiedColor=<rgba 0 0 0 0>}
       cornerRadii=<4 4 4 4>
      }
      tintColor {colorUseRule=UseColor_Specified, specifiedColor=<rgba 0.75 0 0 1>}
      uVRegion {bIsValid=false, max=<0 0>, min=<0 0>}
     }
     hoveredForeground {colorUseRule=UseColor_ColorTable, specifiedColor=<rgba 0 0 0 0>}
     hoveredSlateSound {resourceObject=null}
     normal {
      bIsDynamicallyLoaded=false
      drawAs=RoundedBox
      imageType=NoImage
      mirroring=NoMirror
      resourceName=None
      resourceObject=null
      tiling=NoTile
      imageSize=<32 32>
      margin=<ltrb 0 0 0 0>
      outlineSettings {
       bUseBrushTransparency=false
       roundingType=FixedRadius
       width=1
       color {colorUseRule=UseColor_ColorTable, specifiedColor=<rgba 0 0 0 0>}
       cornerRadii=<4 4 4 4>
      }
      tintColor {colorUseRule=UseColor_Specified, specifiedColor=<rgba 0.5 0 0 1>}
      uVRegion {bIsValid=false, max=<0 0>, min=<0 0>}
     }
     normalForeground {colorUseRule=UseColor_ColorTable, specifiedColor=<rgba 0 0 0 0>}
     normalPadding=<ltrb 12 1.5 12 1.5>
     pressed {
      bIsDynamicallyLoaded=false
      drawAs=RoundedBox
      imageType=NoImage
      mirroring=NoMirror
      resourceName=None
      resourceObject=null
      tiling=NoTile
      imageSize=<32 32>
      margin=<ltrb 0 0 0 0>
      outlineSettings {
       bUseBrushTransparency=false
       roundingType=FixedRadius
       width=1
       color {colorUseRule=UseColor_ColorTable, specifiedColor=<rgba 0 0 0 0>}
       cornerRadii=<4 4 4 4>
      }
      tintColor {colorUseRule=UseColor_Specified, specifiedColor=<rgba 1 0 0 1>}
      uVRegion {bIsValid=false, max=<0 0>, min=<0 0>}
     }
     pressedForeground {colorUseRule=UseColor_ColorTable, specifiedColor=<rgba 0 0 0 0>}
     pressedPadding=<ltrb 12 2.5 12 0.5>
     pressedSlateSound {resourceObject=null}
    }
   }
   slot=~
  }
  TextBlock_29 {
   widgetClassPath=&TextBlock
   parent=EditorUtilityButton_28
   widget=TextBlock_29
   properties {text=Fix Event Anim Sequence}
   slot {class=ButtonSlot}
  }
 ]
}
